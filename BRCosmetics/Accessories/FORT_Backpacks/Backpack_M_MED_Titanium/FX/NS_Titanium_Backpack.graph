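# Niagara effect for the Titanium backpack cosmetic that flickers lightning bolts and glows on its skeletal mesh.
# Connected_Lightning_Bolts samples indexed bone/socket points to place bolts and separate left and right glow positions.
# Random_Lightning_Bolts adds randomly oriented bolts, with system Timeline blending their lifetime and scale.
# Both combine sprite, mesh and light renderers with curve-driven color flashes and front-end-aware material parameters; a debug emitter is disabled.
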NiagaraSystem /BRCosmetics/Accessories/FORT_Backpacks/Backpack_M_MED_Titanium/FX/NS_Titanium_Backpack
dependencies {
 usedDataInterfaces [
  &NiagaraDataInterfaceCurve
  &NiagaraDataInterfaceColorCurve
  &NiagaraDataInterfaceArrayFloat
  &NiagaraDataInterfaceArrayInt32
 ]
 usedDynamicInputs [&SelectIntFromArray, &FNTimeOfDayScalar]
 usedModules [
  &SystemState
  &self:SystemSpawnScript.SystemScriptSource.SystemScriptGraph.NiagaraNodeAssignment_2.SetVariables_#1
  &Timeline
 ]
}
emitters {
 Connected_Lightning_Bolts {
  emitterUpdateScript [
   &EmitterState(Life Cycle Mode=::NewEnumerator1, Loop Duration=1, UseLoopDelay=false)
   &SpawnRate(Spawn Probability=0.3, SpawnRate=3, Use Spawn Probability=true)
  ]
  particleSpawnScript [
   &InitializeParticle(
    Color=&RandomRangeLinearColor(Minimum=<rgba 0.7 0.7 0.7 1>, Maximum=<rgba 1.3 1.3 1.3 1>)
    Lifetime=&RandomRangeFloat(Minimum=0.25, Maximum=0.4)
    Mesh Scale Mode=::NewEnumerator4
    Sprite Rotation Mode=::NewEnumerator2
    Sprite Size Mode=::NewEnumerator3
    Uniform Sprite Size=15
    UsePositionOffset=false
    Write Material Random=true
   )
   /SetVariables() {isSetParameters=true, module=SetVariables_#2}
   /SetVariables() {isSetParameters=true, module=SetVariables_#3}
   /SetVariables(Particles.VisibilityTag=&RandomRangeInt(Minimum=0, Maximum=1)) {isSetParameters=true}
   /SetVariables(Particles.SpriteGlowSize:Vector2f=&Vector2DFromFloat(Value=30)) {isSetParameters=true}
   &DynamicMaterialParameters(
    Index 0 Param 2=&RandomRangeFloat(Minimum=0.2, Maximum=0.5)
    Index 1 Param 2=&FN_FrontEndFloatModifier(Gameplay=1, Front End=0)
    Module.Param0WriteEnabled=false
    Module.Param0WriteEnabled1=false
    Module.Param1WriteEnabled=true
    Module.Param1WriteEnabled1=true
    Module.Param2WriteEnabled=false
    Module.Param2WriteEnabled1=false
    Module.Param3WriteEnabled=false
    Module.Param3WriteEnabled1=false
    Write Parameter Index 1=true
   ) {module=DynamicMaterialParameters001}
  ]
  particleUpdateScript [
   &ParticleState()
   &SkeletalMeshLocation(
    MaskSpawnGroup=false
    Mesh Sampling Type=::NewEnumerator5
    Orientation Sampling=::NewEnumerator1
    Skeletal Mesh=$User.SkeletalMesh
    Socket Sampling Mode=::NewEnumerator5
   )
   &SkeletalMeshLocation(
    Bone / Socket Index=1
    MaskSpawnGroup=false
    Mesh Sampling Type=::NewEnumerator5
    Orientation Sampling=::NewEnumerator1
    Skeletal Mesh=$User.SkeletalMesh
    Socket Sampling Mode=::NewEnumerator5
    Write To Particle Position=false
   ) {module=SkeletalMeshLocation001}
   &SkeletalMeshLocation(
    Bone / Socket Index=2
    MaskSpawnGroup=false
    Mesh Sampling Type=::NewEnumerator5
    Orientation Sampling=::NewEnumerator1
    Skeletal Mesh=$User.SkeletalMesh
    Socket Sampling Mode=::NewEnumerator5
    Write To Particle Position=false
   ) {module=SkeletalMeshLocation002}
   /SetVariables(Particles.PositionGlowLeft:NiagaraPosition=$Particles.SkeletalMeshLocation001.SampledPosition) {isSetParameters=true}
   /SetVariables(Particles.PositionGlowRight:NiagaraPosition=$Particles.SkeletalMeshLocation002.SampledPosition) {isSetParameters=true}
   /SetVariables(Particles.PositionGlowBig:NiagaraPosition=&AddVectorToPosition(Position=$Particles.SkeletalMeshLocation.SampledPosition, Vector=<0 0 10>)) {enabled=false, isSetParameters=true}
   &ScaleSpriteSize(Uniform Curve Sprite Scale=DataInterface() {propertyValues {Curve {keys [<at 0 0.3 RCIM_Cubic>,<at 0.1 1 RCIM_Cubic>]}}})
   &CameraOffset(Camera Offset Amount=5)
   &ScaleMeshSize(Scale Factor=&MakeVector(
     X=&FloatFromCurve(FloatCurve=DataInterface() {propertyValues {Curve {keys [
          {arriveTangent=0.3854, interpMode=RCIM_Cubic, leaveTangent=0.3854, tangentMode=RCTM_User, time=0, value=1}
          {arriveTangent=0.301, interpMode=RCIM_Cubic, leaveTangent=0.301, tangentMode=RCTM_User, time=0.7, value=1.8}
         ]}}})
     Y=1
     Z=1
    )) {enabled=false}
   /SetVariables(Particles.ColorGlow:LinearColor=&MakeLinearColorFromVectorAndFloat(
     "Vector (RGB)"=&MakeVectorFromLinearColorRGB(Color=&ColorFromCurve(ColorCurve=DataInterface() {propertyValues {
         AlphaCurve {keys []}
         BlueCurve {keys [<at 0.0908 100>,<at 0.3027 1 RCIM_Cubic>]}
         GreenCurve {keys [<at 0.0908 100>,<at 0.3027 1 RCIM_Cubic>]}
         RedCurve {keys [<at 0.0908 100>,<at 0.3027 1 RCIM_Cubic>]}
        }}))
     "Float (Alpha)"=&FloatFromCurve(
      FloatCurve=DataInterface() {propertyValues {Curve {keys [
          {arriveTangent=31.1377, interpMode=RCIM_Cubic, leaveTangent=31.1377, tangentMode=RCTM_User, time=0, value=0}
          {arriveTangent=0.007, interpMode=RCIM_Cubic, leaveTangent=0.007, tangentMode=RCTM_User, time=0.05, value=0.5}
          {arriveTangent=-0.0192, interpMode=RCIM_Cubic, leaveTangent=-0.0192, tangentMode=RCTM_User, time=0.2, value=0.5}
          <at 0.8 0 RCIM_Cubic>
         ]}}}
      Scale Curve=0.01
     )
    )) {isSetParameters=true}
   /SetVariables(Particles.ColorSprite:LinearColor=&MakeLinearColorFromVectorAndFloat(
     "Vector (RGB)"=&MakeVectorFromLinearColorRGB(Color=&ColorFromCurve(ColorCurve=DataInterface() {propertyValues {
         AlphaCurve {keys []}
         BlueCurve {keys [<at 0.0908 40>,<at 0.1816 1 RCIM_Cubic>]}
         GreenCurve {keys [<at 0.0908 40>,<at 0.1816 1 RCIM_Cubic>]}
         RedCurve {keys [<at 0.0908 40>,<at 0.1816 1 RCIM_Cubic>]}
        }}))
     "Float (Alpha)"=&FloatFromCurve(
      FloatCurve=DataInterface() {propertyValues {Curve {keys [
          {arriveTangent=0.0039, interpMode=RCIM_Cubic, leaveTangent=0.0039, tangentMode=RCTM_User, time=0.1, value=0}
          {arriveTangent=0.007, interpMode=RCIM_Cubic, leaveTangent=0.007, tangentMode=RCTM_User, time=0.2, value=1}
          {arriveTangent=-0.0192, interpMode=RCIM_Cubic, leaveTangent=-0.0192, tangentMode=RCTM_User, time=0.3, value=1}
          <at 0.8 0 RCIM_Cubic>
         ]}}}
      Scale Curve=0.5
     )
    )) {isSetParameters=true, module=SetVariables_#4}
   &ScaleColor(
    Scale Alpha=&FloatFromCurve(FloatCurve=DataInterface() {propertyValues {Curve {keys [<at 0.8 2>,<at 1 0>]}}})
    Scale RGB=&MakeVectorFromLinearColorRGB(Color=&ColorFromCurve(ColorCurve=DataInterface() {propertyValues {
        AlphaCurve {keys []}
        BlueCurve {keys [<at 0 1 RCIM_Cubic>,<at 0.2659 0.2>]}
        GreenCurve {keys [<at 0 1 RCIM_Cubic>,<at 0.2659 0.2>]}
        RedCurve {keys [<at 0 1 RCIM_Cubic>,<at 0.2659 0.2>]}
       }}))
    ScaleA=true
    ScaleRGB=true
   )
   &DynamicMaterialParameters(
    Index 0 Param 2=0.4
    Index 0 Param 3=8
    Index 0 Param 4=&FloatFromCurve(
     FloatCurve=DataInterface() {propertyValues {Curve {keys [
         {arriveTangent=0.1537, interpMode=RCIM_Cubic, leaveTangent=0.1537, tangentMode=RCTM_User, time=0.6, value=0}
         {arriveTangent=4.1515, interpMode=RCIM_Cubic, leaveTangent=4.1515, tangentMode=RCTM_User, time=1, value=1.525}
        ]}}}
     Scale Curve=0.2
    )
    Index 1 Param 1=&FloatFromCurve(
     FloatCurve=DataInterface() {propertyValues {Curve {keys [
         {arriveTangent=0.1804, interpMode=RCIM_Cubic, leaveTangent=0.1804, tangentMode=RCTM_User, time=0.5, value=0}
         {arriveTangent=3.1151, interpMode=RCIM_Cubic, leaveTangent=3.1151, tangentMode=RCTM_User, time=0.9, value=0.3}
        ]}}}
     Scale Curve=0.03
    )
    Module.Param0WriteEnabled=false
    Module.Param0WriteEnabled1=true
    Module.Param1WriteEnabled=false
    Module.Param1WriteEnabled1=false
    Module.Param2WriteEnabled=true
    Module.Param2WriteEnabled1=false
    Module.Param3WriteEnabled=true
    Module.Param3WriteEnabled1=false
    Write Parameter Index 1=true
   )
  ]
  rendererClasses [
   &NiagaraSpriteRendererProperties
   &NiagaraLightRendererProperties
   &NiagaraMeshRendererProperties
  ]
  renderers [
   {rendererClass=&NiagaraSpriteRendererProperties}
   {rendererClass=&NiagaraLightRendererProperties, rendererIndex=1}
   {rendererClass=&NiagaraMeshRendererProperties, rendererIndex=2}
   {rendererClass=&NiagaraMeshRendererProperties, rendererIndex=3}
   {rendererClass=&NiagaraSpriteRendererProperties, rendererIndex=4}
   {rendererClass=&NiagaraSpriteRendererProperties, rendererIndex=5}
   {rendererClass=&NiagaraSpriteRendererProperties, rendererIndex=6}
  ]
 }
 FN_Debug {
  bEnabled=false
  emitterUpdateScript [
   &EmitterState()
   &SpawnBurst_Instantaneous(Spawn Count=1, Use Spawn Probability=false)
  ]
  particleSpawnScript [/Niagara/Modules/Spawn/Initialization/InitializeParticle(
    Lifetime=5
    Module.Write Color=true
    Module.Write Lifetime=true
    Module.Write Mass=true
    Module.Write Material Random=true
    Module.Write Position=false
    Module.Write Scale=false
    Module.Write SpriteRotation=false
    Module.Write SpriteSize=true
    Position=&SimulationPosition()
    Sprite Size=<100 100>
   )]
  particleUpdateScript [
   &ParticleState()
   &FN_Timer()
   /SetVariables(Particles.Debug:NiagaraFloat=$System.FN_Timer.Value) {isSetParameters=true}
   &DynamicMaterialParameters(
    Index 0 Param 1=$Particles.Debug
    Module.Param0WriteEnabled=true
    Module.Param1WriteEnabled=true
    Module.Param2WriteEnabled=true
    Module.Param3WriteEnabled=true
   )
   &CameraOffset(Camera Offset Amount=111)
  ]
  rendererClasses [&NiagaraSpriteRendererProperties]
  renderers [{rendererClass=&NiagaraSpriteRendererProperties}]
 }
 Random_Lightning_Bolts {
  emitterUpdateScript [
   &EmitterState(Life Cycle Mode=::NewEnumerator1, Loop Duration=1, UseLoopDelay=false)
   &SpawnRate(Spawn Probability=0.15, SpawnRate=10, Use Spawn Probability=true)
  ]
  particleSpawnScript [
   /Niagara/Modules/Spawn/Initialization/V2/InitializeParticle(
    Color=&RandomRangeLinearColor(Minimum=<rgba 0.7 0.7 0.7 1>, Maximum=<rgba 1.3 1.3 1.3 1>)
    Lifetime=&Lerp_Float(
     A=&RandomRangeFloat(Minimum=0.3, Maximum=0.5)
     B=&RandomRangeFloat(Minimum=0.2, Maximum=0.3)
     Alpha=$System.Timeline.NormalizedTime
    )
    Mesh Scale=&Lerp_Vector(
     A=&RandomRangeVector(Minimum=<0.6 0.6 0.6>, Maximum=<1.5 1.5 1.3>)
     B=&RandomRangeVector(Minimum=<1 1 1.3>, Maximum=<2 2 2>)
     Alpha=$System.Timeline.NormalizedTime
    )
    Mesh Scale Mode=::NewEnumerator4
    Sprite Rotation Mode=::NewEnumerator2
    Sprite Size Mode=::NewEnumerator3
    Uniform Sprite Size=15
    UsePositionOffset=false
    Write Material Random=true
   )
   &CameraOffset(Camera Offset Amount=5)
   /SetVariables(Particles.OffsetMesh=4) {isSetParameters=true}
   /SetVariables() {isSetParameters=true, module=SetVariables_#2}
   /SetVariables(Particles.VisibilityTag=&RandomRangeInt(Minimum=0, Maximum=2)) {isSetParameters=true}
   /SetVariables() {isSetParameters=true, module=SetVariables_#3}
   &DynamicMaterialParameters(
    Index 0 Param 1=&FN_FrontEndFloatModifier(Gameplay=1, Front End=0)
    Index 0 Param 4=&SelectFloatFromArray(
     Direct Array Index=$Particles.VisibilityTag
     Direct Array Mode=::NewEnumerator1
     Float Selection Array=DataInterface() {propertyValues {FloatData [0.5,1.5,1.5]}}
    )
    Module.Param0WriteEnabled=true
    Module.Param1WriteEnabled=false
    Module.Param2WriteEnabled=false
    Module.Param3WriteEnabled=true
   )
  ]
  particleUpdateScript [
   &ParticleState()
   &SkeletalMeshLocation(
    MaskSpawnGroup=false
    Mesh Sampling Type=::NewEnumerator5
    Orientation Sampling=::NewEnumerator1
    OverrideRandomSeed=false
    Skeletal Mesh=$User.SkeletalMesh
   )
   &ScaleSpriteSize(Uniform Curve Sprite Scale=DataInterface() {propertyValues {Curve {keys [<at 0 0.3 RCIM_Cubic>,<at 0.2 1 RCIM_Cubic>]}}})
   &ScaleMeshSize(Scale Factor=&MakeVector(
     X=&FloatFromCurve(FloatCurve=DataInterface() {propertyValues {Curve {keys [
          {arriveTangent=1.0254, interpMode=RCIM_Cubic, leaveTangent=1.0254, tangentMode=RCTM_User, time=0, value=0.7}
          {interpMode=RCIM_Cubic, tangentMode=RCTM_User, time=0.5, value=1}
         ]}}})
     Y=&FloatFromCurve(FloatCurve=DataInterface() {propertyValues {Curve {keys [
          {arriveTangent=1.0254, interpMode=RCIM_Cubic, leaveTangent=1.0254, tangentMode=RCTM_User, time=0, value=0.7}
          {interpMode=RCIM_Cubic, tangentMode=RCTM_User, time=0.5, value=1}
         ]}}})
     Z=&FloatFromCurve(FloatCurve=DataInterface() {propertyValues {Curve {keys [
          {arriveTangent=0.8974, interpMode=RCIM_Cubic, leaveTangent=0.8974, tangentMode=RCTM_User, time=0, value=0.5}
          {interpMode=RCIM_Cubic, tangentMode=RCTM_User, time=0.75, value=1}
         ]}}})
    ))
   &UpdateMeshOrientation(
    Facing Direction=&RotateVectorByQuaternion(
     Vector To Rotate=&RandomRangeVector(Minimum=<-1 -1 -1>, Maximum=<1 1 1>)
     Quaternion=$Particles.SkeletalMeshLocation.SampledRotation
    )
    Facing Interpolation=false
    Orientation Method=::NewEnumerator1
    Override Quat=false
    TransformFacing=false
    TransformUp=false
   ) {module=UpdateMeshOrientation001}
   &ScaleColor(
    Scale Alpha=&FloatFromCurve(
     FloatCurve=DataInterface() {propertyValues {Curve {keys [
         {arriveTangent=14.472, interpMode=RCIM_Cubic, leaveTangent=14.472, tangentMode=RCTM_User, time=0.12, value=0.5}
         {arriveTangent=-0.0363, interpMode=RCIM_Cubic, leaveTangent=-0.0363, tangentMode=RCTM_User, time=0.17, value=1}
         <at 0.95 1>
         <at 1 0>
        ]}}}
     Scale Curve=3
    )
    Scale RGB=&MakeVectorFromLinearColorRGB(Color=&ColorFromCurve(ColorCurve=DataInterface() {propertyValues {
        AlphaCurve {keys []}
        BlueCurve {keys [<at 0.0146 1 RCIM_Cubic>,<at 0.1743 0.06>]}
        GreenCurve {keys [<at 0.0146 1 RCIM_Cubic>,<at 0.1743 0.06>]}
        RedCurve {keys [<at 0.0146 1 RCIM_Cubic>,<at 0.1743 0.06>]}
       }}))
    ScaleA=true
    ScaleRGB=true
   ) {module=ScaleColor001}
   /SetVariables(Particles.ColorSprite:LinearColor=&MakeLinearColorFromVectorAndFloat(
     "Vector (RGB)"=&MakeVectorFromLinearColorRGB(Color=&ColorFromCurve(ColorCurve=DataInterface() {propertyValues {
         AlphaCurve {keys []}
         BlueCurve {keys [<at 0.0443 14>,<at 0.4563 1 RCIM_Cubic>]}
         GreenCurve {keys [<at 0.0443 14>,<at 0.4563 1 RCIM_Cubic>]}
         RedCurve {keys [<at 0.0443 14>,<at 0.4563 1 RCIM_Cubic>]}
        }}))
     "Float (Alpha)"=&FloatFromCurve(
      FloatCurve=DataInterface() {propertyValues {Curve {keys [
          {arriveTangent=31.1377, interpMode=RCIM_Cubic, leaveTangent=31.1377, tangentMode=RCTM_User, time=0, value=0}
          {arriveTangent=-2.0108, interpMode=RCIM_Cubic, leaveTangent=-2.0108, tangentMode=RCTM_User, time=0.05, value=1}
          <at 0.4 0 RCIM_Cubic>
         ]}}}
      Scale Curve=0.1
     )
    )) {isSetParameters=true, module=SetVariables_#4}
   &DynamicMaterialParameters(
    Index 0 Param 2=&FloatFromCurve(
     FloatCurve=DataInterface() {propertyValues {Curve {keys [
         {arriveTangent=0.1337, interpMode=RCIM_Cubic, leaveTangent=0.1337, tangentMode=RCTM_User, time=0.3, value=0}
         {arriveTangent=2.5592, interpMode=RCIM_Cubic, leaveTangent=2.5592, tangentMode=RCTM_User, time=0.9, value=0.5}
        ]}}}
     Scale Curve=0.1
    )
    Index 0 Param 3=&FloatFromCurve(
     FloatCurve=DataInterface() {propertyValues {Curve {keys [
         {arriveTangent=0.1034, interpMode=RCIM_Cubic, leaveTangent=0.1034, tangentMode=RCTM_User, time=0.9, value=0}
         {arriveTangent=1.3775, interpMode=RCIM_Cubic, leaveTangent=1.3775, tangentMode=RCTM_User, time=1, value=1}
        ]}}}
     Scale Curve=0.3
    )
    Index 0 Param 4=1
    Module.Param0WriteEnabled=false
    Module.Param1WriteEnabled=true
    Module.Param2WriteEnabled=true
    Module.Param3WriteEnabled=false
   ) {module=DynamicMaterialParameters001}
  ]
  rendererClasses [
   &NiagaraSpriteRendererProperties
   &NiagaraMeshRendererProperties
   &NiagaraLightRendererProperties
  ]
  renderers [
   {rendererClass=&NiagaraSpriteRendererProperties}
   {rendererClass=&NiagaraMeshRendererProperties, rendererIndex=1}
   {rendererClass=&NiagaraLightRendererProperties, rendererIndex=2}
   {rendererClass=&NiagaraMeshRendererProperties, rendererIndex=3}
   {rendererClass=&NiagaraMeshRendererProperties, rendererIndex=4}
  ]
 }
}
summary {
 systemName=NS_Titanium_Backpack
 emitters [
  {bEnabled=false, emitterName=FN_Debug, rendererClasses [&NiagaraSpriteRendererProperties]}
  {
   emitterName=Connected_Lightning_Bolts
   rendererClasses [
    &NiagaraSpriteRendererProperties
    &NiagaraLightRendererProperties
    &NiagaraMeshRendererProperties
   ]
  }
  {
   emitterName=Random_Lightning_Bolts
   rendererClasses [
    &NiagaraSpriteRendererProperties
    &NiagaraMeshRendererProperties
    &NiagaraLightRendererProperties
   ]
  }
 ]
 userVariables [
  User.bIsFrontEnd {type=NiagaraBool}
  User.bIsFrontEndPreview {type=NiagaraBool, defaultValue=true}
  User.Empowered {type=NiagaraBool}
  User.SkeletalMesh {
   type=NiagaraDataInterfaceSkeletalMesh
   defaultValue {
    struct=NiagaraExt_VariableValue_DataInterface
    value {dataInterface=&self:NiagaraDataInterfaceSkeletalMesh_0}
   }
  }
 ]
}
